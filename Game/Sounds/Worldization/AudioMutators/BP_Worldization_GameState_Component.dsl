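; Game state component that switches the worldization audio mutators for SFX, music and chat on or off, on cosmetic-capable instances only.
; BeginPlay binds to the worldization subsystem's OnWorldizationEnabled event and seeds the initial state from IsAGVDisabled, noted as a check of AGV rollback.
; The update handler adds or removes all three mutators through the Audio Mutators subsystem.
; EndPlay removes them, then calls Assign on the same event again rather than unbinding.
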(event EventBeginPlay
  (Networking|CanExecuteCosmeticEvents
    (:True
      (AudioWorldization|AssignOnWorldizationEnabled (WorldSubsystems|GetAudioWorldizationSubsystem) (EventDispatchers|CreateEvent :!event Custom|UpdateWorldizationMutators))
      ;[0 0 0 0.5] Looks into current AGV rollback
      (CallFunction|UpdateWorldizationMutators (AudioWorldization|IsAGVDisabled (WorldSubsystems|GetAudioWorldizationSubsystem))))))

(event Custom|UpdateWorldizationMutators
  (params (bIsEnabled :type Boolean))
  (bind _returnvalue (AudioEngineSubsystems|GetAudioMutatorsSubsystem))
  (bind _returnvalue_2 (AudioEngineSubsystems|GetAudioMutatorsSubsystem))
  (bind _returnvalue_3 (AudioEngineSubsystems|GetAudioMutatorsSubsystem))
  (if bIsEnabled
    (Audio|Mutator|AddMutator _returnvalue (Variables|Mutators|GetWorldizationMutatorSFX))
    (Audio|Mutator|AddMutator _returnvalue_2 (Variables|Mutators|GetWorldizationMutatorMusic))
    (Audio|Mutator|AddMutator _returnvalue_3 (Variables|Mutators|GetWorldizationMutatorChat))
    (else
      (Audio|Mutator|RemoveMutator _returnvalue (Variables|Mutators|GetWorldizationMutatorSFX))
      (Audio|Mutator|RemoveMutator _returnvalue_2 (Variables|Mutators|GetWorldizationMutatorMusic))
      (Audio|Mutator|RemoveMutator _returnvalue_3 (Variables|Mutators|GetWorldizationMutatorChat)))))

(event EventEndPlay
  (params (EndPlayReason :type EEndPlayReason%20Enum :default "Destroyed"))
  (Networking|CanExecuteCosmeticEvents
    (:True
      (CallFunction|UpdateWorldizationMutators false)
      (AudioWorldization|AssignOnWorldizationEnabled (WorldSubsystems|GetAudioWorldizationSubsystem) (EventDispatchers|CreateEvent :!event Custom|UpdateWorldizationMutators)))))
